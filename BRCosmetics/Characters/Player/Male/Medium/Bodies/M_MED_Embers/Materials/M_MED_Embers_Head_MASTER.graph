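# Master head material for the male medium Embers character body, using material attributes on skeletal meshes with Default Lit and Subsurface Profile shading.
# Shared character PreFX and PostFX functions wrap a World Position Offset edit.
# A sphere mask on a selectable vertex color channel collapses vertices near a parameterized value to the object position, effectively hiding them.
# A quality switch skips the collapse on Low, keeping the PreFX offset.
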Material /BRCosmetics/Characters/Player/Male/Medium/Bodies/M_MED_Embers/Materials/M_MED_Embers_Head_MASTER
graph {
 properties {
  bUseMaterialAttributes=true
  bUsedWithMeshDeformer=true
  bUsedWithSkeletalMesh=true
  pixelDepthOffsetMode=PDOM_Legacy
  shadingModel=MSM_FromMaterialExpression
  usedShadingModels=Default Lit | Subsurface Profile
 }
 nodes [
  n1 = MaterialFunctionCall() {materialFunction=&BaseFN_CharacterFunction_PreFX}
  n2 = GetMaterialAttributes(None=@n1.Result) {attributeGetTypes [#1]}
  n3 = ObjectPositionWS()
  n4 = WorldPosition()
  n5 = ChannelMaskParameter(None=VertexColor()) {parameterName=VertexColorChannel}
  n6 = ScalarParameter() {defaultValue=0.5, parameterName=ColorMaskValue}
  n7 = ScalarParameter() {defaultValue=0.01, parameterName=Threshold}
  n8 = Constant() {r=1}
  n9 = Add(A=@n2.World Position Offset, B=LinearInterpolate(B=Subtract(A=@n3, B=@n4.XYZ), Alpha=SphereMask(A=@n5.R, B=@n6, Radius=@n7, Hardness=@n8)))
  MaterialFunctionCall(
   AO=@n1.AO
   Distance Blend=@n1.AO
   Fresnel=@n1.AO
   Gmap=@n1.AO
   Pre Skinned Local Position=@n1.AO
   Material=SetMaterialAttributes!(
    MaterialAttributes=@n2.MaterialAttributes
    World Position Offset=QualitySwitch(Default=@n9, Low=@n2.World Position Offset, High=@n9, Medium=@n9, Epic=@n9)
   ) {attributeSetTypes [#1]}
  ) {materialFunction=&BaseFN_CharacterFunction_PostFX}
 ]
}
