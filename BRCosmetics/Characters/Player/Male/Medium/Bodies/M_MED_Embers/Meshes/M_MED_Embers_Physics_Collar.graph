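# Physics asset for the collar of the M_MED_Embers male medium player body, setting up constrained secondary motion on its dyn_collar bones.
# Five flat box bodies, offset along each bone, form the back, left and right mid, and left and right collar segments.
# Each segment is constrained only to its own small master-bone box, with no links between neighbors.
# All joints share narrow, asymmetric swing limits and locked twist.
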PhysicsAsset /BRCosmetics/Characters/Player/Male/Medium/Bodies/M_MED_Embers/Meshes/M_MED_Embers_Physics_Collar
bodies {
 dyn_collar1_l [{shapeType=Box, extentX=5, extentY=5, extentZ=1, center=<4 0 0>}]
 dyn_collar1_mstr_l [{shapeType=Box, extentX=1.01, extentY=1.01, extentZ=1.01}]
 dyn_collar1_mstr_r [{shapeType=Box, extentX=1.01, extentY=1.01, extentZ=1.01}]
 dyn_collar1_r [{shapeType=Box, extentX=5, extentY=5, extentZ=1, center=<4 0 0>}]
 dyn_collarBack1 [{shapeType=Box, extentX=5, extentY=6, extentZ=1, center=<4 0 0>}]
 dyn_collarBack_mstr [{shapeType=Box, extentX=1.01, extentY=1.01, extentZ=1.01}]
 dyn_collarMid1_l [{shapeType=Box, extentX=5, extentY=6, extentZ=1, center=<4 0 0>}]
 dyn_collarMid1_mstr_l [{shapeType=Box, extentX=1.01, extentY=1.01, extentZ=1.01}]
 dyn_collarMid1_mstr_r [{shapeType=Box, extentX=1.01, extentY=1.01, extentZ=1.01}]
 dyn_collarMid1_r [{shapeType=Box, extentX=5, extentY=6, extentZ=1, center=<4 0 0>}]
}
constraints [
 {bone1Name=dyn_collar1_l, bone2Name=dyn_collar1_mstr_l, swing1LimitDegrees=5, swing2LimitDegrees=18, twistLimitDegrees=0, twistMotion=Locked}
 {bone1Name=dyn_collar1_r, bone2Name=dyn_collar1_mstr_r, swing1LimitDegrees=5, swing2LimitDegrees=18, twistLimitDegrees=0, twistMotion=Locked}
 {
  bone1Name=dyn_collarMid1_r
  bone2Name=dyn_collarMid1_mstr_r
  swing1LimitDegrees=5
  swing2LimitDegrees=18
  twistLimitDegrees=0
  twistMotion=Locked
 }
 {
  bone1Name=dyn_collarBack1
  bone2Name=dyn_collarBack_mstr
  swing1LimitDegrees=5
  swing2LimitDegrees=18
  twistLimitDegrees=0
  twistMotion=Locked
 }
 {
  bone1Name=dyn_collarMid1_l
  bone2Name=dyn_collarMid1_mstr_l
  swing1LimitDegrees=5
  swing2LimitDegrees=18
  twistLimitDegrees=0
  twistMotion=Locked
 }
]
