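# Physics asset for the left and right coat lapels on the male medium M_MED_Embers character body.
# Each side has a thin, offset box body on the lapel bone and a small box body on its master bone.
# One constraint per side joins the lapel to its master, with twist locked and small swing limit angles.
# The two sides mirror each other's setup and have no constraint between them.
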PhysicsAsset /BRCosmetics/Characters/Player/Male/Medium/Bodies/M_MED_Embers/Meshes/M_MED_Embers_Physics_Lapels
bodies {
 dyn_lapel1_l [{shapeType=Box, extentX=8, extentY=4, extentZ=1, center=<5 0 0>}]
 dyn_lapel1_mstr_l [{shapeType=Box, extentX=1.01, extentY=1.01, extentZ=1.01}]
 dyn_lapel1_mstr_r [{shapeType=Box, extentX=1.01, extentY=1.01, extentZ=1.01}]
 dyn_lapel1_r [{shapeType=Box, extentX=8, extentY=4, extentZ=1, center=<5 0 0>}]
}
constraints [
 {bone1Name=dyn_lapel1_l, bone2Name=dyn_lapel1_mstr_l, swing1LimitDegrees=5, swing2LimitDegrees=5, twistLimitDegrees=15, twistMotion=Locked}
 {bone1Name=dyn_lapel1_r, bone2Name=dyn_lapel1_mstr_r, swing1LimitDegrees=5, swing2LimitDegrees=5, twistLimitDegrees=15, twistMotion=Locked}
]
