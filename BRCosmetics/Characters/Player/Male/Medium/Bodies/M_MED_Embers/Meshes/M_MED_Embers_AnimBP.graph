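# Post-copy animation Blueprint for the M_MED_Embers male body mesh, adding secondary motion to its lapels and collar.
# It copies the attached parent's pose through a Control Rig, then simulates lapels and collar with two RigidBody nodes.
# A per-update flag for skydive floating or diving fades the Control Rig out and fades in additive offsets on three collar bones, using two-second linear blends.
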AnimBlueprint /BRCosmetics/Characters/Player/Male/Medium/Bodies/M_MED_Embers/Meshes/M_MED_Embers_AnimBP
target_skeleton=&/M_MED_Embers_Skeleton
graphs [
 AnimGraph {nodes [
   n1 = Get isSkyDivingAny() -> (isSkyDivingAny:bool)
   n2 = Control Rig!(Source=Copy Pose From Mesh() {bUseAttachedParent=true}, bAlphaBoolEnabled:bool=NOT Boolean(A=@n1)) -> (Pose) {
    alphaInputType=Bool
    alphaBoolBlend {blendInTime=2, blendOption=Linear, blendOutTime=2}
    controlRigAssetReference {blueprintRigClass=/BRCosmetics/Characters/Player/Male/Medium/Bodies/M_MED_Embers/Meshes/M_MED_Embers_CtrlRig/M_MED_Embers_CtrlRig.M_MED_Embers_CtrlRig_C}
   }
   n3 = RigidBody(ComponentPose=Local To Component(LocalPose=@n2)) {
    bClampLinearTranslationLimitToRefPose=true
    bOverrideWorldGravity=true
    overridePhysicsAsset=&/M_MED_Embers_Physics_Lapels
    simulationSpace=BaseBoneSpace
    alphaBoolBlend {blendOption=Linear}
    baseBoneRef {boneName=pelvis}
    componentAppliedLinearAccClamp=<5000 5000 5000>
    componentLinearAccScale=<0.4 0.4 0.4>
    componentLinearVelScale=<0.3 0.3 0.3>
    overrideWorldGravity {z=-980}
   }
   n4 = RigidBody(ComponentPose=@n3) {
    bClampLinearTranslationLimitToRefPose=true
    bOverrideWorldGravity=true
    overridePhysicsAsset=&/M_MED_Embers_Physics_Collar
    simulationSpace=BaseBoneSpace
    alphaBoolBlend {blendOption=Linear}
    baseBoneRef {boneName=spine_05}
    componentAppliedLinearAccClamp=<5000 5000 5000>
    componentLinearAccScale=<0.4 0.4 0.4>
    componentLinearVelScale=<0.3 0.3 0.3>
    overrideWorldGravity {z=-980}
   }
   n5 = "Transform (Modify) Bone - Bone: dyn_collarBack_mstr"!(ComponentPose=@n4, bAlphaBoolEnabled:bool=@n1, Translation="-2,0,-1.5", Rotation="-55,0.000,0", Scale="1,1,1") -> (Pose) {
    alphaInputType=Bool
    rotationMode=BMM_Additive
    rotationSpace=BCS_BoneSpace
    translationMode=BMM_Additive
    translationSpace=BCS_BoneSpace
    alphaBoolBlend {blendInTime=2, blendOption=Linear, blendOutTime=2}
    boneToModify {boneName=dyn_collarBack_mstr}
    rotation {pitch=-55}
    translation {x=-2, z=-1.5}
   }
   n6 = "Transform (Modify) Bone - Bone: dyn_collarMid1_mstr_l"!(ComponentPose=@n5, bAlphaBoolEnabled:bool=@n1, Translation="-1,0,-0.5", Rotation="-55,0,0", Scale="1,1,1") -> (Pose) {
    alphaInputType=Bool
    rotationMode=BMM_Additive
    rotationSpace=BCS_BoneSpace
    translationMode=BMM_Additive
    translationSpace=BCS_BoneSpace
    alphaBoolBlend {blendInTime=2, blendOption=Linear, blendOutTime=2}
    boneToModify {boneName=dyn_collarMid1_mstr_l}
    rotation {pitch=-55}
    translation {x=-1, z=-0.5}
   }
   n7 = "Transform (Modify) Bone - Bone: dyn_collarMid1_mstr_r"!(ComponentPose=@n6, bAlphaBoolEnabled:bool=@n1, Translation="-1,0,-0.5", Rotation="-55,0,0", Scale="1,1,1") -> (Pose) {
    alphaInputType=Bool
    rotationMode=BMM_Additive
    rotationSpace=BCS_BoneSpace
    translationMode=BMM_Additive
    translationSpace=BCS_BoneSpace
    alphaBoolBlend {blendInTime=2, blendOption=Linear, blendOutTime=2}
    boneToModify {boneName=dyn_collarMid1_mstr_r}
    rotation {pitch=-55}
    translation {x=-1, z=-0.5}
   }
   Output Pose(Result=Component To Local(ComponentPose=@n7)) {layerGroup=None}
  ]}
 EventGraph {nodes [
   Event BlueprintUpdateAnimation()
   > Set isSkyDivingAny(isSkyDivingAny:bool=OR Boolean(A=Get bIsSkydiveFloating(), B=Get bIsSkydiveDiveMode())) -> (Output_Get:bool)
  ]}
]
variables [isSkyDivingAny {type=boolean}]
